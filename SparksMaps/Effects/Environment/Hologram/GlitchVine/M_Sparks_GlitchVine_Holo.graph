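# Additive unlit hologram material for the GlitchVine environment effect in Sparks, tinted by the stage primary color from MF_MojoStageColors.
# Samples the vine texture three times with sine-jittered UV offsets for RGB channel splitting, adding a vertical gradient glow.
# Tiled, flippable UVs are distorted by panned texture noise, with glitch strength switched by layered sine time.
# Scanlines and a sine/gradient mask modulate emissive, alongside view-angle and depth fades.
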Material /SparksMaps/Effects/Environment/Hologram/GlitchVine/M_Sparks_GlitchVine_Holo
graph {
 properties {bAutomaticallySetUsageInEditor=false, blendMode=BLEND_Additive, shadingModel=MSM_Unlit, usedShadingModels=Unlit}
 nodes [
  n1 = Constant() {r=100}
  n2 = Constant() {r=0.025}
  n3 = Constant() {r=0.8}
  n4 = Constant() {r=1}
  n5 = Constant() {r=0.15}
  n6 = Constant() {r=2}
  n7 = ComponentMask(None=TextureCoordinate()) {r=true}
  n8 = Multiply(
   A=Sine(None=Multiply(A=@n7) {constB=0.5})
   B=SmoothStep(Value=MaterialFunctionCall() {materialFunction=&LinearGradient}.VGradient) {constMax=0.75, constMin=1}
  )
  n9 = NamedRerouteUsage()
  n10 = TextureSample(UVs=NamedRerouteUsage()) {
   samplerSource=SSM_Wrap_WorldGroupSettings
   samplerType=SAMPLERTYPE_LinearColor
   texture=/SparksMaps/Effects/Screens/GlitchVine/T_GlitchVine_RGB
  }
  n11 = Time()
  n12 = ScalarParameter() {defaultValue=2.476, parameterName=Param_4}
  n13 = Clamp(None=Sine(None=Multiply(A=@n11, B=@n12) {constB=1.76})) {maxDefault=0.004, minDefault=-0.005}
  n14 = If(A=NamedRerouteUsage(), "A > B"=Constant() {r=-0.01}, "A < B"=Constant() {r=-0.1})
  n15 = TextureSample(UVs=Add(A=NamedRerouteUsage(), B=AppendVector(A=@n13, B=@n14))) {
   samplerSource=SSM_Wrap_WorldGroupSettings
   samplerType=SAMPLERTYPE_LinearColor
   texture=&T_GlitchVine_RGB
  }
  n16 = Time()
  n17 = ScalarParameter() {defaultValue=1.345, parameterName=Param_3}
  n18 = Clamp(None=Sine(None=Multiply(A=@n16, B=@n17) {constB=2.43})) {maxDefault=0.005, minDefault=-0.006}
  n19 = ScalarParameter() {defaultValue=0.01, parameterName=Param_2}
  n20 = TextureSample(UVs=Add(A=NamedRerouteUsage(), B=AppendVector(A=@n18, B=@n19))) {
   samplerSource=SSM_Wrap_WorldGroupSettings
   samplerType=SAMPLERTYPE_LinearColor
   texture=&T_GlitchVine_RGB
  }
  n21 = MaterialFunctionCall(R=@n10.G, G=@n15.G, B=@n20.G) {materialFunction=&AppendMany}
  n22 = ScalarParameter() {defaultValue=1}
  n23 = Power(
   Base=MaterialFunctionCall() {materialFunction=&LinearGradient}.VGradient
   Exp=ScalarParameter() {defaultValue=3, parameterName=Param_5}
  )
  n24 = Add(A=Multiply(A=@n9, B=Multiply(A=@n21.RGB, B=@n22)), B=Multiply(A=@n9, B=Multiply(A=@n23) {constB=0.5}))
  MaterialFunctionCall(Emissive=Multiply(
    A=MaterialFunctionCall!(
     Number of Scanlines=@n1
     Scanline Speed=@n2
     "Scanlines Thickness (Bias)"
     Scanlines Dark Value=@n3
     Scanlines Light Value=@n4
     Block Opacity=@n5
     UVs
     Time
     Block Repeat=@n6
    ) {materialFunction=/Game/Effects/Fort_Effects/Effects/Atlas/Materials/MF_Atlas_Scanlines_01}.Result
    B=Multiply(A=@n8, B=@n24)
   )) {materialFunction=&MF_FX_EmissiveAdjustments}
  Multiply(
   A=MaterialFunctionCall("FadeSensitivity (optional)"=ScalarParameter() {defaultValue=2, parameterName=Param_6}) {materialFunction=&FadeBasedOnViewAngle}.Result
   B=DepthFade() {fadeDistanceDefault=250}
  )
  NamedRerouteDeclaration(None=Add(A=MaterialFunctionCall!(CustomUPosition) {materialFunction=/SparksStageShow/Systems/StageManager/FXManager/MaterialFunctions/MF_MojoStageColors}.Primary) {constB=0.25}) {name=Primary Color, nodeColor=<rgba 0 0 0 1>}
  n25 = TextureCoordinate()
  n26 = ComponentMask(None=@n25) {r=true}
  n27 = OneMinus(None=@n26)
  n28 = ScalarParameter() {parameterName=FlipX}
  n29 = Multiply(
   A=AppendVector(A=LinearInterpolate(A=@n26, B=@n27, Alpha=@n28), B=ComponentMask(None=@n25) {g=true})
   B=AppendVector(A=ScalarParameter() {defaultValue=1, parameterName=XTile}, B=ScalarParameter() {defaultValue=0.5, parameterName=YTile})
  )
  n30 = AppendVector(A=ScalarParameter() {parameterName=XPos}, B=ScalarParameter() {defaultValue=0.5, parameterName=YPos})
  n31 = NamedRerouteUsage()
  n32 = Constant() {r=0.1}
  n33 = Constant() {r=0.25}
  n34 = TextureSample(UVs=Panner(Coordinate=TextureCoordinate(), Speed=If(A=@n31, "A > B"=@n32, "A < B"=@n33)) {speedX=0.1}) {
   samplerSource=SSM_Wrap_WorldGroupSettings
   samplerType=SAMPLERTYPE_LinearColor
   texture=&T_GlitchVine_RGB
  }
  n35 = If(A=NamedRerouteUsage(), "A > B"=Constant() {r=0.005}, "A < B"=Constant() {r=0.05})
  NamedRerouteDeclaration(None=Add(A=Add(A=@n29, B=@n30), B=Multiply(A=@n34.R, B=@n35) {constB=0.02})) {name=UVs, nodeColor=<rgba 1 0.2578 0 1>}
  n36 = Add(A=Time(), B=ScalarParameter() {parameterName=GlitchTimeOffset})
  n37 = ScalarParameter() {defaultValue=1, parameterName=Glitch Anim Amount}
  n38 = Multiply(A=Multiply(A=@n36) {constB=0.5}, B=@n37)
  n39 = Multiply(A=Multiply(A=@n36) {constB=0.376}, B=@n37)
  n40 = Multiply(A=Multiply(A=@n36) {constB=0.4}, B=@n37)
  NamedRerouteDeclaration(None=LinearInterpolate(A=Sine(None=@n38), B=Sine(None=@n39), Alpha=Sine(None=@n40))) {name=Glitch Time, nodeColor=<rgba 0 0.1094 1 1>}
  ScalarParameter() {defaultValue=-0.1, parameterName=Param_1}
  VectorParameter() {group=ScreenThemeParams, parameterName=TestColorA, defaultValue=<rgba 0 0.4518 1 1>}
 ]
}
